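# Niagara character effect for the Hightower_Soy skin: a cyan-to-blue glow trailing along both arms and hands of the user-supplied skeletal mesh.
# Active HandsGlowMesh emitter spawns two mirrored groups, sampling bones or sockets twice to store separate arm and hand positions for two mesh renderers.
# A scalability-weighted random value splits particles into large faint and small bright, short-lived ones, driving offsets and material parameters.
# Sprite-based HandsGlowExpensive emitter is disabled.
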NiagaraSystem /BRCosmetics/Effects/Fort_Effects/Effects/Characters/Athena_Parts/HighTower_Soy/NS_Hightower_Soy_Skin_Glow
dependencies {usedDataInterfaces [&NiagaraDataInterfaceCurve], usedDynamicInputs [], usedModules [&SystemState]}
emitters {
 HandsGlowExpensive {
  bEnabled=false
  emitterUpdateScript [
   &EmitterState()
   &SpawnRate(SpawnRate=25, Use Spawn Probability=false)
   &SpawnRate(SpawnGroup=1, SpawnRate=25, Use Spawn Probability=false) {module=SpawnRate001}
  ]
  particleSpawnScript [
   /SetVariables(
    Particles.CameraOffsetInitRandom:NiagaraFloat=&UniformRangedFloat(Minimum=0, Maximum=1)
    Particles.RandomPlacement:Vector3f=&UniformRangedVector(Minimum=<1 -1 -1>, Maximum=<1 1 1>)
    Particles.RandomTearaway:NiagaraFloat=&UniformRangedFloat(Minimum=0, Maximum=1)
   ) {isSetParameters=true, module=SetVariables_#1}
   /SetVariables(Particles.RandomPlacementLength:NiagaraFloat=&Multiply_Float(
     A=0.5774
     B=&VectorLength(Vector=&Add_Vector(
       A=&UniformRangedVector(Minimum=<-1 0 0>, Maximum=<1 0 0>)
       B=&Multiply_Vector(A=<0 1 1>, B=$Particles.RandomPlacement)
      ))
    )) {isSetParameters=true, module=SetVariables_#2}
   /SetVariables(Particles.CameraOffsetRandom:NiagaraFloat=&FloatFromCurve(
     FloatCurve=DataInterface() {propertyValues {Curve {keys [{arriveTangent=-2.8433, interpMode=RCIM_Cubic, leaveTangent=-2.8433, tangentMode=RCTM_User, time=0.3, value=1}, <at 0.5 0>]}}}
     CurveIndex=$Particles.RandomPlacementLength
    )) {isSetParameters=true}
   /Niagara/Modules/Spawn/Initialization/InitializeParticle(
    Color=&MakeLinearColor(
     R:NiagaraFloat=0.226
     G:NiagaraFloat=0.8157
     B:NiagaraFloat=1
     A:NiagaraFloat=&Lerp_Float(A=0.2, B=0.5, Alpha=$Particles.CameraOffsetRandom)
    )
    Lifetime=&UniformRangedFloat(Minimum=1, Maximum=2)
    Module.Write Color=true
    Module.Write Lifetime=true
    Module.Write Mass=true
    Module.Write Material Random=true
    Module.Write Position=true
    Module.Write Scale=false
    Module.Write SpriteRotation=true
    Module.Write SpriteSize=true
    Position=&SimulationPosition()
    Sprite Rotation=&UniformRangedFloat(Minimum=0, Maximum=360)
    Sprite Size=&Vector2DFromFloat(Value=&Lerp_Float(A=&UniformRangedFloat(Minimum=5, Maximum=8), B=20, Alpha=$Particles.RandomTearaway))
   )
   /NiagaraFN/Sprite/FN_RandomizeSpriteUV()
   &CameraOffset(Camera Offset Amount=&Lerp_Float(A=0, B=5, Alpha=$Particles.CameraOffsetRandom))
   &DynamicMaterialParameters(
    Index 0 Param 3=&UniformRangedFloat(Minimum=0, Maximum=1)
    Module.Param0WriteEnabled=false
    Module.Param1WriteEnabled=false
    Module.Param2WriteEnabled=true
    Module.Param3WriteEnabled=false
   )
  ]
  particleUpdateScript [
   &ParticleState()
   /SetVariables(
    Particles.BoneGroupVector:Vector3f=&MaskVectorBySpawnGroup(Masked Vector=<-1 -1 -1>, Passthrough Vector=<1 1 1>)
    Particles.InitialLocation:Vector3f=&Multiply_Vector(A=<30 1 1>, B=$Particles.RandomPlacement)
    Particles.Offset:Vector3f=&Multiply_VectorByFloat(
     Vector=&Add_Vector(
      A=&Lerp_Vector(A=<-30 0 0>, B=<-6 0 0>, Alpha=$Particles.RandomTearaway)
      B=&Multiply_Vector(
       A=<0 3 3>
       B=&Multiply_Vector(
        A=$Particles.RandomPlacement
        B=&Add_Vector(
         A=<0 1 1>
         B=&Multiply_VectorByFloat(Vector=&Multiply_Vector(A=$Particles.RandomPlacement, B=$Particles.RandomPlacement), Float=-0.7)
        )
       )
      )
     )
     Float=$Particles.NormalizedAge
    )
   ) {isSetParameters=true, module=SetVariables_#3}
   &SkeletalMeshLocation(
    Apply Sampled Orientation to Mesh Particles=true
    Bone / Socket Index=&SpawnGroup()
    Bone Sampling Mode=::NewEnumerator4
    Inherit Velocity=false
    MaskSpawnGroup=false
    Sampled Position Offset=&Multiply_Vector(A=$Particles.BoneGroupVector, B=&Add_Vector(A=$Particles.InitialLocation, B=$Particles.Offset))
    Sampled Position Offset Coordinate Space:ENiagaraMeshTransforms=::NewEnumerator3
    Skeletal Mesh=$User.SkeletalMesh
   )
   &ScaleColor(
    Scale Alpha=&FloatFromCurve(FloatCurve=DataInterface() {propertyValues {Curve {keys [
         {arriveTangent=3.0697, interpMode=RCIM_Cubic, leaveTangent=3.0697, tangentMode=RCTM_User, time=0, value=0}
         {arriveTangent=2.7272, interpMode=RCIM_Cubic, leaveTangent=2.7272, tangentMode=RCTM_User, time=0.1358, value=0.6998}
         {arriveTangent=0.002, interpMode=RCIM_Cubic, leaveTangent=0.002, tangentMode=RCTM_User, time=0.3, value=1}
         {arriveTangent=-0.9169, interpMode=RCIM_Cubic, leaveTangent=-0.9169, tangentMode=RCTM_User, time=1.0011, value=0.002}
        ]}}})
    Scale RGB=&Lerp_Vector(A=<1 1 1>, B=&MakeVectorFromLinearColorRGB(Color=<rgba 0 0.4358 1 1>), Alpha=$Particles.NormalizedAge)
    ScaleA=true
    ScaleRGB=true
   )
   &ScaleSpriteSize(Scale Factor=&Vector2DFromFloat(Value=&FloatFromCurve(FloatCurve=DataInterface() {propertyValues {Curve {keys [
          {arriveTangent=0.7209, interpMode=RCIM_Cubic, leaveTangent=0.7209, tangentMode=RCTM_User, time=-0.0046, value=0.7145}
          {arriveTangent=0.1347, interpMode=RCIM_Cubic, leaveTangent=0.1347, tangentMode=RCTM_User, time=0.6218, value=1}
          {arriveTangent=-0.2593, interpMode=RCIM_Cubic, leaveTangent=-0.2593, tangentMode=RCTM_User, time=1.0011, value=0.85}
         ]}}})))
   &Color(Color=&LerpLinearColors(StartColor=<rgba 0 0 0.786 1>, EndColor=<rgba 1 0 0 1>, LerpFactor=$Particles.CameraOffsetRandom)) {enabled=false}
  ]
  rendererClasses [&NiagaraSpriteRendererProperties]
  renderers [{rendererClass=&NiagaraSpriteRendererProperties}]
 }
 HandsGlowMesh {
  emitterUpdateScript [
   &EmitterState()
   &SpawnRate(SpawnRate=15, Use Spawn Probability=false)
   &SpawnRate(SpawnGroup=1, SpawnRate=15, Use Spawn Probability=false) {module=SpawnRate001}
  ]
  particleSpawnScript [
   /SetVariables(
    Particles.RandomPlacement:Vector3f=&UniformRangedVector(Minimum=<1 -1 -1>, Maximum=<1 1 1>)
    Particles.RandomTearaway:NiagaraFloat=&UniformRangedFloat(Minimum=0, Maximum=1)
    Particles.ScalabilityLerp:NiagaraFloat=&Add_Float(A=-0.25, B=&Multiply_Float(A=1.25, B=$Engine.Emitter.SpawnCountScale))
   ) {isSetParameters=true, module=SetVariables_#1}
   /SetVariables(Particles.RandomPlacementLength:NiagaraFloat=&ClampFloat(Float=&Multiply_Float(
      A=0.5774
      B=&VectorLength(Vector=&Add_Vector(
        A=&UniformRangedVector(Minimum=<-1 0 0>, Maximum=<1 0 0>)
        B=&Multiply_Vector(A=<0 1 1>, B=$Particles.RandomPlacement)
       ))
     ))) {isSetParameters=true, module=SetVariables_#2}
   /SetVariables(Particles.CameraOffsetRandom:NiagaraFloat=&ClampFloat(Float=&Ceil_Float(Float=&Add_Float(
       A=&Add_Float(A=-0.15, B=&Multiply_Float(A=-0.35, B=$Particles.ScalabilityLerp))
       B=&Multiply_Float(A=$Particles.ScalabilityLerp, B=$Particles.RandomPlacementLength)
      )))) {isSetParameters=true}
   &InitializeParticle(
    Color=&MakeLinearColor(
     R:NiagaraFloat=0.226
     G:NiagaraFloat=0.8157
     B:NiagaraFloat=1
     A:NiagaraFloat=&Lerp_Float(A=0.05, B=7, Alpha=$Particles.CameraOffsetRandom)
    )
    Lifetime=&Multiply_Float(
     A=&Lerp_Float(A=2, B=3, Alpha=$Particles.RandomTearaway)
     B=&Lerp_Float(A=1, B=0.25, Alpha=$Particles.CameraOffsetRandom)
    )
    Mesh Scale=&VectorFromFloat(Value=&Multiply_Float(
      A=&Lerp_Float(A=1, B=0.3, Alpha=$Particles.CameraOffsetRandom)
      B=&Lerp_Float(A=&UniformRangedFloat(Minimum=5, Maximum=8), B=20, Alpha=$Particles.RandomTearaway)
     ))
    Module.Write Color=true
    Module.Write Lifetime=true
    Module.Write Mass=true
    Module.Write Material Random=true
    Module.Write Position=true
    Module.Write Scale=true
    Module.Write SpriteRotation=true
    Module.Write SpriteSize=false
    Position=&SimulationPosition()
    Sprite Rotation=&UniformRangedFloat(Minimum=0, Maximum=360)
   )
   /SetVariables(Particles.ScaleAlt:Vector3f=&VectorFromFloat(Value=&Multiply_Float(
      A=&Lerp_Float(A=1, B=0.4, Alpha=$Particles.CameraOffsetRandom)
      B=&UniformRangedFloat(Minimum=10, Maximum=12)
     ))) {isSetParameters=true}
   &CameraOffset(Camera Offset Amount=&Lerp_Float(A=0, B=10, Alpha=$Particles.CameraOffsetRandom))
   /SetVariables(
    Particles.MeshOrientation=&MultiplyQuaternion(
     Quaternion A=&OrientMeshToVector(Look At Direction=<0 0 -1>, Reference Axis Vector=<1 0 0>)
     Quaternion B=&MakeQuatFromAxisAngle(Axis=<0 0 1>, AngleInDegrees=&UniformRangedFloat(Minimum=-360, Maximum=360))
    )
    Particles.MeshOrientationAlt:Quat4f=&MultiplyQuaternion(
     Quaternion A=&OrientMeshToVector(Look At Direction=<0 0 -1>, Reference Axis Vector=<1 0 0>)
     Quaternion B=&MakeQuatFromAxisAngle(Axis=<0 0 1>, AngleInDegrees=&UniformRangedFloat(Minimum=-360, Maximum=360))
    )
   ) {isSetParameters=true}
  ]
  particleUpdateScript [
   &ParticleState()
   &DynamicMaterialParameters(
    Index 0 Param 1=&Multiply_Float(A=0.2, B=$Engine.Time)
    Index 0 Param 2=$Particles.RandomTearaway
    Index 0 Param 3=1
    Index 0 Param 4=&Lerp_Float(A=0.1, B=4, Alpha=$Particles.CameraOffsetRandom)
    Module.Param0WriteEnabled=true
    Module.Param1WriteEnabled=true
    Module.Param2WriteEnabled=true
    Module.Param3WriteEnabled=true
   )
   /SetVariables(
    Particles.BoneGroupVector:Vector3f=&MaskVectorBySpawnGroup(Masked Vector=<-1 -1 -1>, Passthrough Vector=<1 1 1>)
    Particles.InitialLocation:Vector3f=&Multiply_Vector(A=<22 0.5 2>, B=$Particles.RandomPlacement)
    Particles.Offset:Vector3f=&Multiply_VectorByFloat(
     Vector=&Add_Vector(
      A=&Lerp_Vector(A=<-25 0 0>, B=<-6 0 0>, Alpha=$Particles.RandomTearaway)
      B=&Multiply_Vector(
       A=$Particles.RandomPlacement
       B=&Multiply_Vector(
        A=<0 1 1>
        B=&Add_Vector(
         A=<1.5 1.5 1.5>
         B=&Multiply_VectorByFloat(Vector=&Multiply_Vector(A=$Particles.RandomPlacement, B=$Particles.RandomPlacement), Float=-0.5)
        )
       )
      )
     )
     Float=$Particles.NormalizedAge
    )
   ) {isSetParameters=true, module=SetVariables_#3}
   &SkeletalMeshLocation(
    Bone / Socket Index=&SpawnGroup()
    Bone Sampling Mode=::NewEnumerator4
    Inherit Velocity=false
    MaskSpawnGroup=false
    Sampled Position Offset=&Multiply_Vector(
     A=$Particles.BoneGroupVector
     B=&Add_Vector(
      A=&Multiply_Vector(A=$Particles.InitialLocation, B=&Lerp_Vector(A=<1 1 1>, B=<1 0 0>, Alpha=$Particles.CameraOffsetRandom))
      B=&Multiply_Vector(A=$Particles.Offset, B=&Lerp_Vector(A=<1 1 1>, B=<0.5 0 0>, Alpha=$Particles.CameraOffsetRandom))
     )
    )
    Sampled Position Offset Coordinate Space:ENiagaraMeshTransforms=::NewEnumerator3
    Skeletal Mesh=$User.SkeletalMesh
   )
   /SetVariables(Particles.PositionArms:NiagaraPosition=$Particles.Position) {isSetParameters=true}
   /SetVariables(Particles.Position=$Particles.Initial.Position) {isSetParameters=true}
   /SetVariables(Particles.DynamicMaterialParameterAlt:Vector4f=&MakeVector4(
     X=0
     Y=&Add_Float(A=0.5, B=$Particles.RandomTearaway)
     Z=1
     W=&Lerp_Float(A=0.1, B=4, Alpha=$Particles.CameraOffsetRandom)
    )) {isSetParameters=true}
   /SetVariables(
    Particles.BoneGroupVector:Vector3f=&MaskVectorBySpawnGroup(Masked Vector=<-1 -1 -1>, Passthrough Vector=<1 1 1>)
    Particles.InitialLocation:Vector3f=&Multiply_Vector(A=<10 0 2>, B=$Particles.RandomPlacement)
    Particles.Offset:Vector3f=&Multiply_VectorByFloat(
     Vector=&Add_Vector(
      A=&Lerp_Vector(A=<-15 0 0>, B=<-3 0 0>, Alpha=$Particles.RandomTearaway)
      B=&Multiply_Vector(
       A=$Particles.RandomPlacement
       B=&Multiply_Vector(
        A=<0 1 1>
        B=&Add_Vector(
         A=<1.5 1.5 1.5>
         B=&Multiply_VectorByFloat(Vector=&Multiply_Vector(A=$Particles.RandomPlacement, B=$Particles.RandomPlacement), Float=-0.5)
        )
       )
      )
     )
     Float=$Particles.NormalizedAge
    )
   ) {isSetParameters=true}
   &SkeletalMeshLocation(
    Bone / Socket Index=&Add_Integer(A=&SpawnGroup(), B=2)
    Bone Sampling Mode=::NewEnumerator4
    Inherit Velocity=false
    MaskSpawnGroup=false
    Sampled Position Offset=&Multiply_Vector(
     A=$Particles.BoneGroupVector
     B=&Add_Vector(
      A=&Multiply_Vector(A=$Particles.InitialLocation, B=&Lerp_Vector(A=<1 1 1>, B=<0.5 0.5 0.5>, Alpha=$Particles.CameraOffsetRandom))
      B=&Multiply_Vector(A=$Particles.Offset, B=&Lerp_Vector(A=<1 1 1>, B=<0.1 0 0>, Alpha=$Particles.CameraOffsetRandom))
     )
    )
    Sampled Position Offset Coordinate Space:ENiagaraMeshTransforms=::NewEnumerator3
    Skeletal Mesh=$User.SkeletalMesh
   ) {module=SkeletalMeshLocation001}
   /SetVariables(Particles.PositionHands:NiagaraPosition=$Particles.Position) {isSetParameters=true}
   &ScaleColor(
    Scale Alpha=&FloatFromCurve(FloatCurve=DataInterface() {propertyValues {Curve {keys [
         {arriveTangent=4.7744, interpMode=RCIM_Cubic, leaveTangent=4.7744, tangentMode=RCTM_User, time=0, value=0}
         {arriveTangent=-0.2362, interpMode=RCIM_Cubic, leaveTangent=-0.2362, tangentMode=RCTM_User, time=0.5827, value=0.9932}
         {arriveTangent=-3.3493, interpMode=RCIM_Cubic, leaveTangent=-3.3493, tangentMode=RCTM_User, time=1.0011, value=0.002}
        ]}}})
    Scale RGB=&Lerp_Vector(A=<1 1 1>, B=&MakeVectorFromLinearColorRGB(Color=<rgba 0 0.2833 1 1>), Alpha=$Particles.NormalizedAge)
    ScaleA=true
    ScaleRGB=true
   )
   /SetVariables(Particles.ScaleAlt:Vector3f=&Multiply_VectorByFloat(
     Vector=$Particles.Initial.Scale
     Float=&FloatFromCurve(FloatCurve=DataInterface() {propertyValues {Curve {keys [
          {arriveTangent=0.7209, interpMode=RCIM_Cubic, leaveTangent=0.7209, tangentMode=RCTM_User, time=0, value=0.2}
          {arriveTangent=0.1347, interpMode=RCIM_Cubic, leaveTangent=0.1347, tangentMode=RCTM_User, time=1, value=1}
         ]}}})
    )) {isSetParameters=true}
   &ScaleMeshSize(Scale Factor=&VectorFromFloat(Value=&FloatFromCurve(FloatCurve=DataInterface() {propertyValues {Curve {keys [
          {arriveTangent=0.7209, interpMode=RCIM_Cubic, leaveTangent=0.7209, tangentMode=RCTM_User, time=-0.0046, value=0.7145}
          {arriveTangent=0.1347, interpMode=RCIM_Cubic, leaveTangent=0.1347, tangentMode=RCTM_User, time=0.4, value=1}
          {arriveTangent=-0.2593, interpMode=RCIM_Cubic, leaveTangent=-0.2593, tangentMode=RCTM_User, time=1.0011, value=0.85}
         ]}}})))
  ]
  rendererClasses [&NiagaraMeshRendererProperties]
  renderers [
   {rendererClass=&NiagaraMeshRendererProperties, rendererIndex=0}
   {rendererClass=&NiagaraMeshRendererProperties, rendererIndex=1}
  ]
 }
}
summary {
 systemName=NS_Hightower_Soy_Skin_Glow
 emitters [
  {bEnabled=false, emitterName=HandsGlowExpensive, rendererClasses [&NiagaraSpriteRendererProperties]}
  {emitterName=HandsGlowMesh, rendererClasses [&NiagaraMeshRendererProperties]}
 ]
 userVariables [User.SkeletalMesh {
   type=NiagaraDataInterfaceSkeletalMesh
   defaultValue {
    struct=NiagaraExt_VariableValue_DataInterface
    value {dataInterface=&self:NiagaraDataInterfaceSkeletalMesh_0}
   }
  }]
}
